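# Translucent UI material that places one TextureSlice texture into the Index-selected grid cell of a sprite sheet, apparently for atlas building.
# A custom HLSL node derives the grid from TextureSize, SliceSize and Padding, then remaps UVs into that cell.
# Saturated UVs sample the slice, and a second custom node's bounds mask multiplies its alpha to zero outside the cell.
# Padding affects only the row and column counts, without insetting the slice.
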Material /TextureAtlasBuilder/M_SpriteSheetBuild
graph {
 properties {blendMode=BLEND_Translucent, materialDomain=MD_UI, pixelDepthOffsetMode=PDOM_Legacy, shadingModel=MSM_Unlit, usedShadingModels=Unlit}
 nodes [
  MaterialExpressionVectorParameter_2 = VectorParameter() {parameterName=TextureSize, defaultValue=<rgba 1024 1024 0 1>}
  MaterialExpressionVectorParameter_3 = VectorParameter() {parameterName=SliceSize, defaultValue=<rgba 144 144 0 1>}
  MaterialExpressionScalarParameter_3 = ScalarParameter() {parameterName=Padding}
  MaterialExpressionScalarParameter_2 = ScalarParameter() {defaultValue=5.5228, parameterName=Index}
  MaterialExpressionTextureCoordinate_1 = TextureCoordinate()
  MaterialExpressionCustom_1 = Custom!(
   TextureSize=@MaterialExpressionVectorParameter_2.RGB
   SpriteSize=@MaterialExpressionVectorParameter_3.RGB
   Padding=@MaterialExpressionScalarParameter_3
   SpriteIndex=@MaterialExpressionScalarParameter_2
   TexCoord=@MaterialExpressionTextureCoordinate_1
  ) {
   code="float2 SpriteTotalSize = SpriteSize + Padding*2;\r\nfloat SpritesPerRow = floor(TextureSize.x / SpriteTotalSize.x);\r\nfloat SpritesPerColumn = floor(TextureSize.y / SpriteTotalSize.y);\r\n\r\nint ColumnIndex = SpriteIndex % SpritesPerRow;\r\nint RowIndex = floor(SpriteIndex / SpritesPerRow);\r\n\r\nfloat2 Scale = float2(1.0 / SpritesPerRow, 1.0 / SpritesPerColumn);\r\nfloat2 Offset = float2(ColumnIndex, RowIndex);\r\n\r\nreturn TexCoord / Scale - Offset;"
   outputType=CMOT_Float2
   showCode=true
   inputs [
    {
     inputName=TextureSize
     input {expression=&self:MaterialExpressionVectorParameter_2, mask=1, maskB=1, maskG=1, maskR=1}
    }
    {
     inputName=SpriteSize
     input {expression=&self:MaterialExpressionVectorParameter_3, mask=1, maskB=1, maskG=1, maskR=1}
    }
    {inputName=Padding, input {expression=&self:MaterialExpressionScalarParameter_3}}
    {inputName=SpriteIndex, input {expression=&self:MaterialExpressionScalarParameter_2}}
    {inputName=TexCoord, input {expression=&self:MaterialExpressionTextureCoordinate_1}}
   ]
  }
  n1 = Custom!(UVs=@MaterialExpressionCustom_1) {
   code="if (UVs.x > 1.0 || UVs.x < 0.0 || UVs.y > 1.0 || UVs.y < 0.0)\r\n    return 0;\r\nreturn 1;"
   outputType=CMOT_Float1
   showCode=true
   inputs [{inputName=UVs, input {expression=&self:MaterialExpressionCustom_1}}]
  }
  Multiply(A=TextureSampleParameter2D(UVs=Saturate(None=@MaterialExpressionCustom_1)) {parameterName=TextureSlice}.A, B=@n1)
 ]
}
