EditorUtilityBlueprint /TextureAtlasBuilder/AtlasBuilder
graphs [
 Build {nodes [
   Build() {is=title K2Node_FunctionEntry}
   n1 = Is Empty!(TargetArray:Array of Texture Object References=Get Texture Slices() -> (Texture Slices:Array of Texture Object References)) -> (ReturnValue:bool)
   n2 = Get Texture Atlas() -> (Texture Atlas:Texture 2D Object Reference)
   n3 = Get MID Build() -> (MID Build:Material Instance Dynamic Object Reference)
   > Branch(Condition=AND Boolean!(A=NOT Boolean(A=@n1), B=IsValid(Object=@n2), C:bool=IsValid(Object=@n3)) -> (ReturnValue))
   n4 = Get Texture Size() -> (Texture Size:int=0)
   > n5 = CreateCanvasRenderTarget2D!(
    execute
    WorldContextObject:Object Reference
    CanvasRenderTarget2DClass:Canvas Render Target 2D Class Reference=/Script/Engine.CanvasRenderTarget2D
    Width:int=@n4
    Height:int=@n4
   ) -> (then, ReturnValue:Canvas Render Target 2D Object Reference)
   > Set CompressionSettings(CompressionSettings=TC_Default, self=@n5)
   > Set Current RT(Current RT:Texture Render Target 2D Object Reference=@n5) -> (Output_Get:Texture Render Target 2D Object Reference)
   n6 = "To Float (Integer)"(InInt=@n4)
   n7 = Get MID Build()
   > SetVectorParameterValue(
    self=@n7
    ParameterName=TextureSize
    Value=Make Linear Color!(R:float=@n6, G:float=@n6, B:float=0, A:float=0) -> (LinearColor:Linear Color Structure) {is=title K2Node_MakeStruct}
   )
   n8 = "To Float (Integer)"(InInt=Get Slice Size() -> (Slice Size:int=0))
   > SetVectorParameterValue(self=@n7, ParameterName=SliceSize, Value=Make Linear Color(R=@n8, G=@n8, B=0, A=0))
   > n9 = BeginDrawCanvasToRenderTarget!(
    execute
    WorldContextObject:Object Reference
    TextureRenderTarget:Texture Render Target 2D Object Reference=Get Current RT() -> (Current RT:Texture Render Target 2D Object Reference)
   ) -> (then, Canvas:Canvas Object Reference, Size:Vector 2D Structure, Context:Draw To Render Target Context Structure)
   > Set Current Canvas(Current Canvas:Canvas Object Reference=@n9) -> (Output_Get:Canvas Object Reference)
   > n10 = For Each Loop(Array:Array of Texture Object References=Get Texture Slices()) -> (Array Element:Texture Object Reference)
   n11 = Sequence(execute=@n10.Completed)
   > EndDrawCanvasToRenderTarget!(execute, WorldContextObject:Object Reference, Context:"Draw To Render Target Context Structure (by ref)"=@n9.Context) -> (then)
   n12 = Get Texture Atlas()
   ConvertRenderTargetToTexture2DEditorOnly(execute=@n11.then_1, RenderTarget=Get Current RT(), Texture=@n12)
   > Set SRGB(SRGB=true, self=@n12)
   > Set CompressionSettings(CompressionSettings=TC_Default, self=@n12)
   > SaveLoadedAssets(AssetsToSave=Make Array("[0]":Object Reference=@n12) -> (Array:Array of Object References), bOnlyIfIsDirty=true)
   n13 = Set Current Texture(execute=@n10.LoopBody, Current Texture:Texture Object Reference=@n10) -> (Output_Get:Texture Object Reference)
   n14 = Get MID Build()
   > SetTextureParameterValue(self=@n14, ParameterName=TextureSlice, Value=@n13)
   > SetScalarParameterValue(self=@n14, ParameterName=Index, Value="To Float (Integer)"(InInt=@n10.Array Index))
   > Draw Material!(
    execute
    self:Canvas Object Reference=Get Current Canvas() -> (Current Canvas:Canvas Object Reference)
    RenderMaterial:Material Interface Object Reference=Get MID Build()
    ScreenPosition:Vector 2D Structure
    ScreenSize:Vector 2D Structure=(X=1024,Y=1024)
    CoordinatePosition:Vector 2D Structure
    CoordinateSize:Vector 2D Structure=(X=1,Y=1)
    Rotation:float=0
    PivotPoint:Vector 2D Structure=(X=0.5,Y=0.5)
   ) -> (then)
  ]}
 EventGraph {nodes [
   Event BeginExecution()
   > n1 = CreateDynamicMaterialInstance(Parent=&/M_SpriteSheetBuild, OptionalName=MID Build, CreationFlags=Transient)
   > Set MID Build(MID Build:Material Instance Dynamic Object Reference=@n1) -> (Output_Get:Material Instance Dynamic Object Reference)
   > Build() {is=title K2Node_CallFunction}
   > FinishExecutingTask()
  ]}
]
variables [
 MID Build {type=MaterialInstanceDynamic}
 Slice Size {type=integer, default=128}
 Texture Atlas {type=Texture2D, default=&/ExampleOutput}
 Texture Size {type=integer, default=1024}
 Texture Slices {
  type="array<Texture>"
  default [
   null
   null
   null
   null
   null
   null
   null
   null
   null
   null
   null
   null
   null
   null
   null
   null
   null
   null
   null
   null
   null
   null
   /Game/UI/HUD/Textures/STW_SquadComms/Icon_StW_Ready
   /Game/UI/HUD/Textures/STW_SquadComms/Icon_StW_FollowMe
   /Game/UI/HUD/Textures/STW_SquadComms/Icon_StW_Hold_Position
   null
   null
   null
   null
   null
   null
   null
   null
   null
   null
   null
   null
   null
   null
   null
  ]
 }
]
