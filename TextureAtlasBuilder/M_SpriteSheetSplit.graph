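# Translucent unlit UI material that selects one sprite by Index from a grid-packed TextureAtlas texture parameter.
# A Custom HLSL node derives the sprite's normalized UV rectangle from texture size, slice size, padding and index, filling rows left to right.
# UVRemap_0-1_ToRange maps the incoming UVs into that rectangle for the atlas sample.
# Padding widens the cell stride on both sides, but the rectangle starts at the cell corner without a padding offset.
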Material /TextureAtlasBuilder/M_SpriteSheetSplit
graph {
 properties {
  bCastRayTracedShadows=false
  blendMode=BLEND_Translucent
  materialDomain=MD_UI
  pixelDepthOffsetMode=PDOM_Legacy
  shadingModel=MSM_Unlit
  usedShadingModels=Unlit
 }
 nodes [
  MaterialExpressionVectorParameter_0 = VectorParameter() {parameterName=TextureSize, defaultValue=<rgba 1024 1024 0 1>}
  MaterialExpressionVectorParameter_1 = VectorParameter() {parameterName=SliceSize, defaultValue=<rgba 144 144 0 1>}
  MaterialExpressionScalarParameter_1 = ScalarParameter() {parameterName=Padding}
  MaterialExpressionScalarParameter_0 = ScalarParameter() {parameterName=Index}
  n1 = Custom!(
   TextureSize=@MaterialExpressionVectorParameter_0.RGB
   SpriteSize=@MaterialExpressionVectorParameter_1.RGB
   Padding=@MaterialExpressionScalarParameter_1
   SpriteIndex=@MaterialExpressionScalarParameter_0
  ) {
   code="float2 SpriteTotalSize = SpriteSize + Padding*2;\r\nint SpritesPerRow = floor(TextureSize.x / SpriteTotalSize.x);\r\n\r\nint ColumnIndex = SpriteIndex % SpritesPerRow;\r\nint RowIndex = floor(SpriteIndex / SpritesPerRow);\r\n\r\nfloat PixelX = ColumnIndex * SpriteTotalSize.x;\r\nfloat PixelY = RowIndex * SpriteTotalSize.y;\r\n\r\nfloat PixelX_End = PixelX + SpriteSize.x;\r\nfloat PixelY_End = PixelY + SpriteSize.y;\r\n\r\nreturn float4(PixelX, PixelX_End, PixelY, PixelY_End) / float4(TextureSize.x, TextureSize.x, TextureSize.y, TextureSize.y);"
   outputType=CMOT_Float4
   showCode=true
   inputs [
    {
     inputName=TextureSize
     input {expression=&self:MaterialExpressionVectorParameter_0, mask=1, maskB=1, maskG=1, maskR=1}
    }
    {
     inputName=SpriteSize
     input {expression=&self:MaterialExpressionVectorParameter_1, mask=1, maskB=1, maskG=1, maskR=1}
    }
    {inputName=Padding, input {expression=&self:MaterialExpressionScalarParameter_1}}
    {inputName=SpriteIndex, input {expression=&self:MaterialExpressionScalarParameter_0}}
   ]
  }
  TextureSampleParameter2D(UVs=MaterialFunctionCall!(X range Y range=@n1, UVs=TextureCoordinate()) {materialFunction=/Engine/Functions/Engine_MaterialFunctions01/Coordinates/UVRemap_0-1_ToRange}.UV) {parameterName=TextureAtlas}
 ]
}
