EditorUtilityWidgetBlueprint /TextureAtlasBuilder/TextureAtlasBuilder
graphs [
 Build {nodes [
   Build() {is=title K2Node_FunctionEntry}
   > n1 = GetSelectedAssets()
   > Branch(Condition="integer > integer"(A=Length(TargetArray:Array of Object References=@n1)))
   n2 = Get BuildTextureSizeX() -> (BuildTextureSizeX:Editor Utility Editable Text Object Reference)
   n3 = String To Integer(InString="To String (Text)"(InText="GetText (Editable Text)"!(self:Editable Text Object Reference=@n2) -> (ReturnValue:Text)))
   n4 = Get BuildTextureSizeY() -> (BuildTextureSizeY:Editor Utility Editable Text Object Reference)
   n5 = String To Integer(InString="To String (Text)"(InText="GetText (Editable Text)"(self=@n4)))
   > n6 = CreateCanvasRenderTarget2D!(
    execute
    WorldContextObject:Object Reference
    CanvasRenderTarget2DClass:Canvas Render Target 2D Class Reference=/Script/Engine.CanvasRenderTarget2D
    Width:int=@n3
    Height:int=@n5
   ) -> (then, ReturnValue:Canvas Render Target 2D Object Reference)
   > Set CurrentRT(CurrentRT:Texture Render Target 2D Object Reference=@n6) -> (Output_Get:Texture Render Target 2D Object Reference)
   n7 = Get MID_Build() -> (MID_Build:Material Instance Dynamic Object Reference)
   > SetVectorParameterValue(
    self=@n7
    ParameterName=TextureSize
    Value=Make Linear Color!(R:float="To Float (Integer)"(InInt=@n3), G:float="To Float (Integer)"(InInt=@n5), B:float=0, A:float=0) -> (LinearColor:Linear Color Structure) {is=title K2Node_MakeStruct}
   )
   n8 = Get BuildSizeX() -> (BuildSizeX:Editor Utility Editable Text Object Reference)
   n9 = String To Integer(InString="To String (Text)"(InText="GetText (Editable Text)"(self=@n8)))
   n10 = Get BuildSizeY() -> (BuildSizeY:Editor Utility Editable Text Object Reference)
   n11 = String To Integer(InString="To String (Text)"(InText="GetText (Editable Text)"(self=@n10)))
   > SetVectorParameterValue(self=@n7, ParameterName=SliceSize, Value=Make Linear Color(R="To Float (Integer)"(InInt=@n9), G="To Float (Integer)"(InInt=@n11), B=0, A=0))
   > n12 = BeginDrawCanvasToRenderTarget!(
    execute
    WorldContextObject:Object Reference
    TextureRenderTarget:Texture Render Target 2D Object Reference=Get CurrentRT() -> (CurrentRT:Texture Render Target 2D Object Reference)
   ) -> (then, Canvas:Canvas Object Reference, Size:Vector 2D Structure, Context:Draw To Render Target Context Structure)
   > Set CurrentCanvas(CurrentCanvas:Canvas Object Reference=@n12) -> (Output_Get:Canvas Object Reference)
   > n13 = For Each Loop(Array:Array of Object References=@n1) -> (Array Element:Object Reference)
   n14 = Sequence(execute=@n13.Completed)
   > EndDrawCanvasToRenderTarget!(execute, WorldContextObject:Object Reference, Context:"Draw To Render Target Context Structure (by ref)"=@n12.Context) -> (then)
   n15 = GetSaveDirectory!(execute=@n14.then_1, self, Object:Object Reference=Get CurrentTexture() -> (CurrentTexture:Texture 2D Object Reference)) -> (then, Directory:String)
   n16 = "GetText (Editable Text)"(self=Get BuildTextureAtlasName() -> (BuildTextureAtlasName:Editor Utility Editable Text Object Reference))
   n17 = Select(Option 0:Text=@n16, Option 1:Text="LOCTEXT(\"Texture\")", Index=TextIsEmpty(InText=@n16)) -> (ReturnValue:Text)
   > n18 = TextureAtlasBuilderCreateStaticTextureEditorOnly!(
    execute
    InRenderTarget:Texture Render Target 2D Object Reference=Get CurrentRT()
    InName:String="To String (Text)"(InText=Format Text!(Format="LOCTEXT(\"{Directory}/{TextureName}\")", Directory:String=@n15, TextureName:Text=@n17) -> (Result))
    TextureSettingsSource:Texture 2D Object Reference=Get CurrentTexture()
    bInvertAlphaChannel:bool=true
   ) -> (then, ReturnValue:Texture 2D Object Reference)
   > SaveLoadedAsset(AssetToSave=@n18, bOnlyIfIsDirty=true)
   n19 = Cast To Texture2D(execute=@n13.LoopBody, Object=@n13)
   > n20 = Set CurrentTexture(CurrentTexture:Texture 2D Object Reference=@n19) -> (Output_Get:Texture 2D Object Reference)
   > SetTextureParameterValue(self=@n7, ParameterName=TextureSlice, Value=@n20)
   > SetScalarParameterValue(self=@n7, ParameterName=Index, Value="To Float (Integer)"(InInt=@n13.Array Index))
   > Draw Material!(
    execute
    self:Canvas Object Reference=Get CurrentCanvas() -> (CurrentCanvas:Canvas Object Reference)
    RenderMaterial:Material Interface Object Reference=Get MID_Build()
    ScreenPosition:Vector 2D Structure
    ScreenSize:Vector 2D Structure=(X=1024,Y=1024)
    CoordinatePosition:Vector 2D Structure
    CoordinateSize:Vector 2D Structure=(X=1,Y=1)
    Rotation:float=0
    PivotPoint:Vector 2D Structure=(X=0.5,Y=0.5)
   ) -> (then)
  ]}
 EventGraph {nodes [
   "On Clicked (SplitButton)"() {is=title K2Node_ComponentBoundEvent}
   > Split!(execute:Exec, self:Self Object Reference) -> (then:Exec)
   "On Clicked (BuildButton)"() {is=title K2Node_ComponentBoundEvent}
   > Build() {is=title K2Node_CallFunction}
   Event Construct()
   > n1 = CreateDynamicMaterialInstance(Parent=&/M_SpriteSheetSplit, OptionalName=MID Split, CreationFlags=Transient)
   > Set MID_Split(MID_Split:Material Instance Dynamic Object Reference=@n1) -> (Output_Get:Material Instance Dynamic Object Reference)
   > n2 = CreateDynamicMaterialInstance(Parent=&/M_SpriteSheetBuild, OptionalName=MID Build, CreationFlags=Transient)
   > Set MID_Build(MID_Build:Material Instance Dynamic Object Reference=@n2) -> (Output_Get:Material Instance Dynamic Object Reference)
  ]}
 GetSaveDirectory {nodes [
   n1 = GetSaveDirectory!() -> (then, Object:Object Reference) {is=title K2Node_FunctionEntry}
   > Return Node!(
    execute
    Directory:String=Split!(InPath:String=Get Object Path String(Object=@n1)) -> (PathPart:String, FilenamePart:String, ExtensionPart:String)
   )
  ]}
 Split {nodes [
   Split() {is=title K2Node_FunctionEntry}
   > n1 = GetSelectedAssets()
   > Branch(Condition="integer > integer"(A=Length(TargetArray=@n1)))
   > n2 = For Each Loop(Array=@n1)
   > n3 = Cast To Texture2D(Object=@n2)
   > n4 = Set CurrentTexture(CurrentTexture=@n3)
   n5 = Get MID_Split() -> (MID_Split:Material Instance Dynamic Object Reference)
   > SetTextureParameterValue(self=@n5, ParameterName=TextureAtlas, Value=@n4)
   n6 = GetSizeX!(self:Texture 2D Object Reference=@n3) -> (ReturnValue:int=0)
   n7 = GetSizeY!(self:Texture 2D Object Reference=@n3) -> (ReturnValue:int=0)
   > SetVectorParameterValue(self=@n5, ParameterName=TextureSize, Value=Make Linear Color(R="To Float (Integer)"(InInt=@n6), G="To Float (Integer)"(InInt=@n7), B=0, A=0))
   n8 = Get SizeX() -> (SizeX:Editor Utility Editable Text Object Reference)
   n9 = String To Integer(InString="To String (Text)"(InText="GetText (Editable Text)"(self=@n8)))
   n10 = Get SizeY() -> (SizeY:Editor Utility Editable Text Object Reference)
   n11 = String To Integer(InString="To String (Text)"(InText="GetText (Editable Text)"(self=@n10)))
   > SetVectorParameterValue(self=@n5, ParameterName=SliceSize, Value=Make Linear Color(R="To Float (Integer)"(InInt=@n9), G="To Float (Integer)"(InInt=@n11), B=0, A=0))
   n12 = "To String (Text)"(InText="GetText (Editable Text)"(self=Get SlicesNum() -> (SlicesNum:Editor Utility Editable Text Object Reference)))
   > n13 = For Loop(LastIndex=int - int(A=String To Integer(InString=@n12)))
   > SetScalarParameterValue(self=@n5, ParameterName=Index, Value="To Float (Integer)"(InInt=@n13))
   > n14 = GetSaveDirectory(Object=Get CurrentTexture()) {is=title K2Node_CallFunction}
   > n15 = CreateCanvasRenderTarget2D(CanvasRenderTarget2DClass=&CanvasRenderTarget2D, Width=@n9, Height=@n11)
   > DrawMaterialToRenderTarget!(
    execute
    WorldContextObject:Object Reference
    TextureRenderTarget:Texture Render Target 2D Object Reference=@n15
    Material:Material Interface Object Reference=Get MID_Split()
   ) -> (then)
   n16 = GetObjectName(Object=Get CurrentTexture())
   > n17 = TextureAtlasBuilderCreateStaticTextureEditorOnly(
    InRenderTarget=@n15
    InName="To String (Text)"(InText=Format Text!(Format="LOCTEXT(\"{Directory}/{TextureName}_Slice{Num}\")", Directory=@n14, TextureName:String=@n16, Num:int=@n13) -> (Result))
    TextureSettingsSource=Get CurrentTexture()
    bInvertAlphaChannel=true
   )
   > SaveLoadedAsset(AssetToSave=@n17, bOnlyIfIsDirty=true)
  ]}
]
variables [
 MID_Build {type=MaterialInstanceDynamic}
 MID_Split {type=MaterialInstanceDynamic}
 Slices {type="array<Texture>", default []}
]
widgets {
 info {rootWidgetClass=&VerticalBox, widgetCount=41}
 widgets [
  VerticalBox_0 {widgetClassPath=&VerticalBox, widget=VerticalBox_0}
  SplitOverlay {widgetClassPath=&Overlay, parent=VerticalBox_0, widget=SplitOverlay, slot {class=VerticalBoxSlot}}
  Background {
   widgetClassPath=&Image
   parent=SplitOverlay
   widget=Background
   properties {colorAndOpacity=<rgba 0.0115 0.0115 0.0115 1>}
   slot {class=OverlaySlot, properties {horizontalAlignment=HAlign_Fill, verticalAlignment=VAlign_Fill}}
  }
  VerticalBox_1 {
   widgetClassPath=&VerticalBox
   parent=SplitOverlay
   widget=VerticalBox_1
   slot {class=OverlaySlot, properties {horizontalAlignment=HAlign_Fill, verticalAlignment=VAlign_Fill, padding=<ltrb 10 10 10 10>}}
  }
  HorizontalBox {widgetClassPath=&HorizontalBox, parent=VerticalBox_1, widget=HorizontalBox, slot {class=VerticalBoxSlot}}
  Slices {
   widgetClassPath=&TextBlock
   parent=HorizontalBox
   widget=Slices
   properties {
    text=Slices
    font {
     bForceMonospaced=false
     bMaterialIsStencil=false
     fontMaterial=null
     fontObject=&Roboto
     letterSpacing=0
     monospacedWidth=1
     size=24
     skewAmount=0
     typefaceFontName=Regular
     outlineSettings {
      bApplyOutlineToDropShadows=false
      bMiteredCorners=false
      bSeparateFillAlpha=false
      outlineMaterial=null
      outlineSize=0
      outlineColor=<rgba 0 0 0 1>
     }
    }
    shadowColorAndOpacity=<rgba 0 0 0 1>
    shadowOffset=<2 2>
   }
   slot {class=HorizontalBoxSlot, properties {padding=<ltrb 5 5 5 5>}}
  }
  Border_16 {
   widgetClassPath=&Border
   parent=HorizontalBox
   widget=Border_16
   properties {
    background {
     bIsDynamicallyLoaded=false
     drawAs=RoundedBox
     imageType=NoImage
     mirroring=NoMirror
     resourceName=None
     resourceObject=null
     tiling=NoTile
     imageSize=<32 32>
     margin=<ltrb 0 0 0 0>
     outlineSettings {
      bUseBrushTransparency=false
      roundingType=HalfHeightRadius
      width=0
      color {colorUseRule=UseColor_Specified, specifiedColor=<rgba 0 0 0 0>}
      cornerRadii=<0 0 0 0>
     }
     tintColor {colorUseRule=UseColor_Specified, specifiedColor=<rgba 1 1 1 1>}
     uVRegion {bIsValid=false, max=<0 0>, min=<0 0>}
    }
    brushColor=<rgba 0.0823 0.0823 0.0823 1>
    padding=<ltrb 12 4 12 2>
   }
   slot {class=HorizontalBoxSlot, properties {padding=<ltrb 5 5 5 5>, size {sizeRule=Fill, value=1}}}
  }
  SlicesNum {
   widgetClassPath=/Script/Blutility.EditorUtilityEditableText
   bIsVariable=true
   parent=Border_16
   widget=SlicesNum
   properties {widgetStyle {
     backgroundImageComposing {
      bIsDynamicallyLoaded=false
      drawAs=Border
      imageType=FullColor
      mirroring=NoMirror
      resourceName=../../../Engine/Content/Slate/Old/HyperlinkDotted.png
      resourceObject=null
      tiling=Both
      imageSize=<0 0>
      margin=<ltrb 0 0 0 0.1875>
      outlineSettings {
       bUseBrushTransparency=false
       roundingType=HalfHeightRadius
       width=0
       color {colorUseRule=UseColor_Specified, specifiedColor=<rgba 0 0 0 0>}
       cornerRadii=<0 0 0 0>
      }
      tintColor {colorUseRule=UseColor_Specified, specifiedColor=<rgba 1 1 1 1>}
      uVRegion {bIsValid=false, max=<0 0>, min=<0 0>}
     }
     backgroundImageSelected {
      bIsDynamicallyLoaded=false
      drawAs=Image
      imageType=FullColor
      mirroring=NoMirror
      resourceName=None
      resourceObject=null
      tiling=NoTile
      imageSize=<2048 2048>
      margin=<ltrb 0 0 0 0>
      outlineSettings {
       bUseBrushTransparency=false
       roundingType=HalfHeightRadius
       width=0
       color {colorUseRule=UseColor_Specified, specifiedColor=<rgba 0 0 0 0>}
       cornerRadii=<0 0 0 0>
      }
      tintColor {colorUseRule=UseColor_Specified, specifiedColor=<rgba 0 0.162 0.7454 1>}
      uVRegion {bIsValid=false, max=<0 0>, min=<0 0>}
     }
     caretImage {
      bIsDynamicallyLoaded=false
      drawAs=Image
      imageType=NoImage
      mirroring=NoMirror
      resourceName=None
      resourceObject=null
      tiling=NoTile
      imageSize=<0 0>
      margin=<ltrb 0 0 0 0>
      outlineSettings {
       bUseBrushTransparency=false
       roundingType=HalfHeightRadius
       width=0
       color {colorUseRule=UseColor_Specified, specifiedColor=<rgba 0 0 0 0>}
       cornerRadii=<0 0 0 0>
      }
      tintColor {colorUseRule=UseColor_Specified, specifiedColor=<rgba 1 1 1 1>}
      uVRegion {bIsValid=false, max=<0 0>, min=<0 0>}
     }
     colorAndOpacity {colorUseRule=UseColor_Foreground, specifiedColor=<rgba 1 0 1 1>}
     font {
      bForceMonospaced=false
      bMaterialIsStencil=false
      fontMaterial=null
      fontObject=null
      letterSpacing=0
      monospacedWidth=1
      size=20
      skewAmount=0
      typefaceFontName=Regular
      outlineSettings {
       bApplyOutlineToDropShadows=false
       bMiteredCorners=false
       bSeparateFillAlpha=false
       outlineMaterial=null
       outlineSize=0
       outlineColor=<rgba 0 0 0 1>
      }
     }
     hintColorAndOpacity {colorUseRule=UseColor_Foreground, specifiedColor=<rgba 1 0 1 1>}
     selectedBackgroundColor {colorUseRule=UseColor_Foreground, specifiedColor=<rgba 1 0 1 1>}
    }}
   slot {class=BorderSlot, properties {padding=<ltrb 12 4 12 2>}}
  }
  HorizontalBox_301 {widgetClassPath=&HorizontalBox, parent=VerticalBox_1, widget=HorizontalBox_301, slot {class=VerticalBoxSlot}}
  TextBlock_112 {
   widgetClassPath=&TextBlock
   parent=HorizontalBox_301
   widget=TextBlock_112
   properties {
    text=Slice Size X
    font {
     bForceMonospaced=false
     bMaterialIsStencil=false
     fontMaterial=null
     fontObject=&Roboto
     letterSpacing=0
     monospacedWidth=1
     size=24
     skewAmount=0
     typefaceFontName=Regular
     outlineSettings {
      bApplyOutlineToDropShadows=false
      bMiteredCorners=false
      bSeparateFillAlpha=false
      outlineMaterial=null
      outlineSize=0
      outlineColor=<rgba 0 0 0 1>
     }
    }
    shadowColorAndOpacity=<rgba 0 0 0 1>
    shadowOffset=<2 2>
   }
   slot {class=HorizontalBoxSlot, properties {padding=<ltrb 5 5 5 5>}}
  }
  Border_15 {
   widgetClassPath=&Border
   parent=HorizontalBox_301
   widget=Border_15
   properties {
    background {
     bIsDynamicallyLoaded=false
     drawAs=RoundedBox
     imageType=NoImage
     mirroring=NoMirror
     resourceName=None
     resourceObject=null
     tiling=NoTile
     imageSize=<32 32>
     margin=<ltrb 0 0 0 0>
     outlineSettings {
      bUseBrushTransparency=false
      roundingType=HalfHeightRadius
      width=0
      color {colorUseRule=UseColor_Specified, specifiedColor=<rgba 0 0 0 0>}
      cornerRadii=<0 0 0 0>
     }
     tintColor {colorUseRule=UseColor_Specified, specifiedColor=<rgba 1 1 1 1>}
     uVRegion {bIsValid=false, max=<0 0>, min=<0 0>}
    }
    brushColor=<rgba 0.0823 0.0823 0.0823 1>
    padding=<ltrb 12 4 12 2>
   }
   slot {class=HorizontalBoxSlot, properties {padding=<ltrb 5 5 5 5>, size {sizeRule=Fill, value=1}}}
  }
  SizeX {
   widgetClassPath=&EditorUtilityEditableText
   bIsVariable=true
   parent=Border_15
   widget=SizeX
   properties {widgetStyle {
     backgroundImageComposing {
      bIsDynamicallyLoaded=false
      drawAs=Border
      imageType=FullColor
      mirroring=NoMirror
      resourceName=../../../Engine/Content/Slate/Old/HyperlinkDotted.png
      resourceObject=null
      tiling=Both
      imageSize=<0 0>
      margin=<ltrb 0 0 0 0.1875>
      outlineSettings {
       bUseBrushTransparency=false
       roundingType=HalfHeightRadius
       width=0
       color {colorUseRule=UseColor_Specified, specifiedColor=<rgba 0 0 0 0>}
       cornerRadii=<0 0 0 0>
      }
      tintColor {colorUseRule=UseColor_Specified, specifiedColor=<rgba 1 1 1 1>}
      uVRegion {bIsValid=false, max=<0 0>, min=<0 0>}
     }
     backgroundImageSelected {
      bIsDynamicallyLoaded=false
      drawAs=Image
      imageType=NoImage
      mirroring=NoMirror
      resourceName=None
      resourceObject=null
      tiling=NoTile
      imageSize=<0 0>
      margin=<ltrb 0 0 0 0>
      outlineSettings {
       bUseBrushTransparency=false
       roundingType=HalfHeightRadius
       width=0
       color {colorUseRule=UseColor_Specified, specifiedColor=<rgba 0 0 0 0>}
       cornerRadii=<0 0 0 0>
      }
      tintColor {colorUseRule=UseColor_Specified, specifiedColor=<rgba 0 0.162 0.7454 1>}
      uVRegion {bIsValid=false, max=<0 0>, min=<0 0>}
     }
     caretImage {
      bIsDynamicallyLoaded=false
      drawAs=Image
      imageType=NoImage
      mirroring=NoMirror
      resourceName=None
      resourceObject=null
      tiling=NoTile
      imageSize=<0 0>
      margin=<ltrb 0 0 0 0>
      outlineSettings {
       bUseBrushTransparency=false
       roundingType=HalfHeightRadius
       width=0
       color {colorUseRule=UseColor_Specified, specifiedColor=<rgba 0 0 0 0>}
       cornerRadii=<0 0 0 0>
      }
      tintColor {colorUseRule=UseColor_Specified, specifiedColor=<rgba 1 1 1 1>}
      uVRegion {bIsValid=false, max=<0 0>, min=<0 0>}
     }
     colorAndOpacity {colorUseRule=UseColor_Foreground, specifiedColor=<rgba 1 0 1 1>}
     font {
      bForceMonospaced=false
      bMaterialIsStencil=false
      fontMaterial=null
      fontObject=null
      letterSpacing=0
      monospacedWidth=1
      size=20
      skewAmount=0
      typefaceFontName=Regular
      outlineSettings {
       bApplyOutlineToDropShadows=false
       bMiteredCorners=false
       bSeparateFillAlpha=false
       outlineMaterial=null
       outlineSize=0
       outlineColor=<rgba 0 0 0 1>
      }
     }
     hintColorAndOpacity {colorUseRule=UseColor_Foreground, specifiedColor=<rgba 1 0 1 1>}
     selectedBackgroundColor {colorUseRule=UseColor_Foreground, specifiedColor=<rgba 1 0 1 1>}
    }}
   slot {class=BorderSlot, properties {padding=<ltrb 12 4 12 2>}}
  }
  TextBlock {
   widgetClassPath=&TextBlock
   parent=HorizontalBox_301
   widget=TextBlock
   properties {
    text=Slice Size Y
    font {
     bForceMonospaced=false
     bMaterialIsStencil=false
     fontMaterial=null
     fontObject=&Roboto
     letterSpacing=0
     monospacedWidth=1
     size=24
     skewAmount=0
     typefaceFontName=Regular
     outlineSettings {
      bApplyOutlineToDropShadows=false
      bMiteredCorners=false
      bSeparateFillAlpha=false
      outlineMaterial=null
      outlineSize=0
      outlineColor=<rgba 0 0 0 1>
     }
    }
    shadowColorAndOpacity=<rgba 0 0 0 1>
    shadowOffset=<2 2>
   }
   slot {class=HorizontalBoxSlot, properties {padding=<ltrb 5 5 5 5>}}
  }
  Border_14 {
   widgetClassPath=&Border
   parent=HorizontalBox_301
   widget=Border_14
   properties {
    background {
     bIsDynamicallyLoaded=false
     drawAs=RoundedBox
     imageType=NoImage
     mirroring=NoMirror
     resourceName=None
     resourceObject=null
     tiling=NoTile
     imageSize=<32 32>
     margin=<ltrb 0 0 0 0>
     outlineSettings {
      bUseBrushTransparency=false
      roundingType=HalfHeightRadius
      width=0
      color {colorUseRule=UseColor_Specified, specifiedColor=<rgba 0 0 0 0>}
      cornerRadii=<0 0 0 0>
     }
     tintColor {colorUseRule=UseColor_Specified, specifiedColor=<rgba 1 1 1 1>}
     uVRegion {bIsValid=false, max=<0 0>, min=<0 0>}
    }
    brushColor=<rgba 0.0823 0.0823 0.0823 1>
    padding=<ltrb 12 4 12 2>
   }
   slot {class=HorizontalBoxSlot, properties {padding=<ltrb 5 5 5 5>, size {sizeRule=Fill, value=1}}}
  }
  SizeY {
   widgetClassPath=&EditorUtilityEditableText
   bIsVariable=true
   parent=Border_14
   widget=SizeY
   properties {widgetStyle {
     backgroundImageComposing {
      bIsDynamicallyLoaded=false
      drawAs=Border
      imageType=FullColor
      mirroring=NoMirror
      resourceName=../../../Engine/Content/Slate/Old/HyperlinkDotted.png
      resourceObject=null
      tiling=Both
      imageSize=<0 0>
      margin=<ltrb 0 0 0 0.1875>
      outlineSettings {
       bUseBrushTransparency=false
       roundingType=HalfHeightRadius
       width=0
       color {colorUseRule=UseColor_Specified, specifiedColor=<rgba 0 0 0 0>}
       cornerRadii=<0 0 0 0>
      }
      tintColor {colorUseRule=UseColor_Specified, specifiedColor=<rgba 1 1 1 1>}
      uVRegion {bIsValid=false, max=<0 0>, min=<0 0>}
     }
     backgroundImageSelected {
      bIsDynamicallyLoaded=false
      drawAs=Image
      imageType=NoImage
      mirroring=NoMirror
      resourceName=None
      resourceObject=null
      tiling=NoTile
      imageSize=<0 0>
      margin=<ltrb 0 0 0 0>
      outlineSettings {
       bUseBrushTransparency=false
       roundingType=HalfHeightRadius
       width=0
       color {colorUseRule=UseColor_Specified, specifiedColor=<rgba 0 0 0 0>}
       cornerRadii=<0 0 0 0>
      }
      tintColor {colorUseRule=UseColor_Specified, specifiedColor=<rgba 0 0.162 0.7454 1>}
      uVRegion {bIsValid=false, max=<0 0>, min=<0 0>}
     }
     caretImage {
      bIsDynamicallyLoaded=false
      drawAs=Image
      imageType=NoImage
      mirroring=NoMirror
      resourceName=None
      resourceObject=null
      tiling=NoTile
      imageSize=<0 0>
      margin=<ltrb 0 0 0 0>
      outlineSettings {
       bUseBrushTransparency=false
       roundingType=HalfHeightRadius
       width=0
       color {colorUseRule=UseColor_Specified, specifiedColor=<rgba 0 0 0 0>}
       cornerRadii=<0 0 0 0>
      }
      tintColor {colorUseRule=UseColor_Specified, specifiedColor=<rgba 1 1 1 1>}
      uVRegion {bIsValid=false, max=<0 0>, min=<0 0>}
     }
     colorAndOpacity {colorUseRule=UseColor_Foreground, specifiedColor=<rgba 1 0 1 1>}
     font {
      bForceMonospaced=false
      bMaterialIsStencil=false
      fontMaterial=null
      fontObject=null
      letterSpacing=0
      monospacedWidth=1
      size=20
      skewAmount=0
      typefaceFontName=Regular
      outlineSettings {
       bApplyOutlineToDropShadows=false
       bMiteredCorners=false
       bSeparateFillAlpha=false
       outlineMaterial=null
       outlineSize=0
       outlineColor=<rgba 0 0 0 1>
      }
     }
     hintColorAndOpacity {colorUseRule=UseColor_Foreground, specifiedColor=<rgba 1 0 1 1>}
     selectedBackgroundColor {colorUseRule=UseColor_Foreground, specifiedColor=<rgba 1 0 1 1>}
    }}
   slot {class=BorderSlot, properties {padding=<ltrb 12 4 12 2>}}
  }
  SplitButton {widgetClassPath=&EditorUtilityButton, parent=VerticalBox_1, widget=SplitButton, slot {class=VerticalBoxSlot}}
  TextBlock_54 {widgetClassPath=&TextBlock, parent=SplitButton, widget=TextBlock_54, properties {text=Split}, slot {class=ButtonSlot}}
  Spacer_98 {widgetClassPath=&Spacer, parent=VerticalBox_0, widget=Spacer_98, properties {size=<1 5>}, slot {class=VerticalBoxSlot}}
  BuildOverlay {widgetClassPath=&Overlay, parent=VerticalBox_0, widget=BuildOverlay, slot {class=VerticalBoxSlot}}
  Background_1 {
   widgetClassPath=&Image
   parent=BuildOverlay
   widget=Background_1
   properties {colorAndOpacity=<rgba 0.0115 0.0115 0.0115 1>}
   slot {class=OverlaySlot, properties {horizontalAlignment=HAlign_Fill, verticalAlignment=VAlign_Fill}}
  }
  VerticalBox {
   widgetClassPath=&VerticalBox
   parent=BuildOverlay
   widget=VerticalBox
   slot {class=OverlaySlot, properties {horizontalAlignment=HAlign_Fill, verticalAlignment=VAlign_Fill, padding=<ltrb 10 10 10 10>}}
  }
  HorizontalBox_2 {widgetClassPath=&HorizontalBox, parent=VerticalBox, widget=HorizontalBox_2, slot {class=VerticalBoxSlot}}
  TextBlock_1 {
   widgetClassPath=&TextBlock
   parent=HorizontalBox_2
   widget=TextBlock_1
   properties {
    text=Slice Size X
    font {
     bForceMonospaced=false
     bMaterialIsStencil=false
     fontMaterial=null
     fontObject=&Roboto
     letterSpacing=0
     monospacedWidth=1
     size=24
     skewAmount=0
     typefaceFontName=Regular
     outlineSettings {
      bApplyOutlineToDropShadows=false
      bMiteredCorners=false
      bSeparateFillAlpha=false
      outlineMaterial=null
      outlineSize=0
      outlineColor=<rgba 0 0 0 1>
     }
    }
    shadowColorAndOpacity=<rgba 0 0 0 1>
    shadowOffset=<2 2>
   }
   slot {class=HorizontalBoxSlot, properties {padding=<ltrb 5 5 5 5>}}
  }
  Border_12 {
   widgetClassPath=&Border
   parent=HorizontalBox_2
   widget=Border_12
   properties {
    background {
     bIsDynamicallyLoaded=false
     drawAs=RoundedBox
     imageType=NoImage
     mirroring=NoMirror
     resourceName=None
     resourceObject=null
     tiling=NoTile
     imageSize=<32 32>
     margin=<ltrb 0 0 0 0>
     outlineSettings {
      bUseBrushTransparency=false
      roundingType=HalfHeightRadius
      width=0
      color {colorUseRule=UseColor_Specified, specifiedColor=<rgba 0 0 0 0>}
      cornerRadii=<0 0 0 0>
     }
     tintColor {colorUseRule=UseColor_Specified, specifiedColor=<rgba 1 1 1 1>}
     uVRegion {bIsValid=false, max=<0 0>, min=<0 0>}
    }
    brushColor=<rgba 0.0823 0.0823 0.0823 1>
    padding=<ltrb 12 4 12 2>
   }
   slot {class=HorizontalBoxSlot, properties {padding=<ltrb 5 5 5 5>, size {sizeRule=Fill, value=1}}}
  }
  BuildSizeX {
   widgetClassPath=&EditorUtilityEditableText
   bIsVariable=true
   parent=Border_12
   widget=BuildSizeX
   properties {widgetStyle {
     backgroundImageComposing {
      bIsDynamicallyLoaded=false
      drawAs=Border
      imageType=FullColor
      mirroring=NoMirror
      resourceName=../../../Engine/Content/Slate/Old/HyperlinkDotted.png
      resourceObject=null
      tiling=Both
      imageSize=<0 0>
      margin=<ltrb 0 0 0 0.1875>
      outlineSettings {
       bUseBrushTransparency=false
       roundingType=HalfHeightRadius
       width=0
       color {colorUseRule=UseColor_Specified, specifiedColor=<rgba 0 0 0 0>}
       cornerRadii=<0 0 0 0>
      }
      tintColor {colorUseRule=UseColor_Specified, specifiedColor=<rgba 1 1 1 1>}
      uVRegion {bIsValid=false, max=<0 0>, min=<0 0>}
     }
     backgroundImageSelected {
      bIsDynamicallyLoaded=false
      drawAs=Image
      imageType=NoImage
      mirroring=NoMirror
      resourceName=None
      resourceObject=null
      tiling=NoTile
      imageSize=<0 0>
      margin=<ltrb 0 0 0 0>
      outlineSettings {
       bUseBrushTransparency=false
       roundingType=HalfHeightRadius
       width=0
       color {colorUseRule=UseColor_Specified, specifiedColor=<rgba 0 0 0 0>}
       cornerRadii=<0 0 0 0>
      }
      tintColor {colorUseRule=UseColor_Specified, specifiedColor=<rgba 0 0.162 0.7454 1>}
      uVRegion {bIsValid=false, max=<0 0>, min=<0 0>}
     }
     caretImage {
      bIsDynamicallyLoaded=false
      drawAs=Image
      imageType=NoImage
      mirroring=NoMirror
      resourceName=None
      resourceObject=null
      tiling=NoTile
      imageSize=<0 0>
      margin=<ltrb 0 0 0 0>
      outlineSettings {
       bUseBrushTransparency=false
       roundingType=HalfHeightRadius
       width=0
       color {colorUseRule=UseColor_Specified, specifiedColor=<rgba 0 0 0 0>}
       cornerRadii=<0 0 0 0>
      }
      tintColor {colorUseRule=UseColor_Specified, specifiedColor=<rgba 1 1 1 1>}
      uVRegion {bIsValid=false, max=<0 0>, min=<0 0>}
     }
     colorAndOpacity {colorUseRule=UseColor_Foreground, specifiedColor=<rgba 1 0 1 1>}
     font {
      bForceMonospaced=false
      bMaterialIsStencil=false
      fontMaterial=null
      fontObject=null
      letterSpacing=0
      monospacedWidth=1
      size=20
      skewAmount=0
      typefaceFontName=Regular
      outlineSettings {
       bApplyOutlineToDropShadows=false
       bMiteredCorners=false
       bSeparateFillAlpha=false
       outlineMaterial=null
       outlineSize=0
       outlineColor=<rgba 0 0 0 1>
      }
     }
     hintColorAndOpacity {colorUseRule=UseColor_Foreground, specifiedColor=<rgba 1 0 1 1>}
     selectedBackgroundColor {colorUseRule=UseColor_Foreground, specifiedColor=<rgba 1 0 1 1>}
    }}
   slot {class=BorderSlot, properties {padding=<ltrb 12 4 12 2>}}
  }
  TextBlock_2 {
   widgetClassPath=&TextBlock
   parent=HorizontalBox_2
   widget=TextBlock_2
   properties {
    text=Slice Size Y
    font {
     bForceMonospaced=false
     bMaterialIsStencil=false
     fontMaterial=null
     fontObject=&Roboto
     letterSpacing=0
     monospacedWidth=1
     size=24
     skewAmount=0
     typefaceFontName=Regular
     outlineSettings {
      bApplyOutlineToDropShadows=false
      bMiteredCorners=false
      bSeparateFillAlpha=false
      outlineMaterial=null
      outlineSize=0
      outlineColor=<rgba 0 0 0 1>
     }
    }
    shadowColorAndOpacity=<rgba 0 0 0 1>
    shadowOffset=<2 2>
   }
   slot {class=HorizontalBoxSlot, properties {padding=<ltrb 5 5 5 5>}}
  }
  Border_11 {
   widgetClassPath=&Border
   parent=HorizontalBox_2
   widget=Border_11
   properties {
    background {
     bIsDynamicallyLoaded=false
     drawAs=RoundedBox
     imageType=NoImage
     mirroring=NoMirror
     resourceName=None
     resourceObject=null
     tiling=NoTile
     imageSize=<32 32>
     margin=<ltrb 0 0 0 0>
     outlineSettings {
      bUseBrushTransparency=false
      roundingType=HalfHeightRadius
      width=0
      color {colorUseRule=UseColor_Specified, specifiedColor=<rgba 0 0 0 0>}
      cornerRadii=<0 0 0 0>
     }
     tintColor {colorUseRule=UseColor_Specified, specifiedColor=<rgba 1 1 1 1>}
     uVRegion {bIsValid=false, max=<0 0>, min=<0 0>}
    }
    brushColor=<rgba 0.0823 0.0823 0.0823 1>
    padding=<ltrb 12 4 12 2>
   }
   slot {class=HorizontalBoxSlot, properties {padding=<ltrb 5 5 5 5>, size {sizeRule=Fill, value=1}}}
  }
  BuildSizeY {
   widgetClassPath=&EditorUtilityEditableText
   bIsVariable=true
   parent=Border_11
   widget=BuildSizeY
   properties {widgetStyle {
     backgroundImageComposing {
      bIsDynamicallyLoaded=false
      drawAs=Border
      imageType=FullColor
      mirroring=NoMirror
      resourceName=../../../Engine/Content/Slate/Old/HyperlinkDotted.png
      resourceObject=null
      tiling=Both
      imageSize=<0 0>
      margin=<ltrb 0 0 0 0.1875>
      outlineSettings {
       bUseBrushTransparency=false
       roundingType=HalfHeightRadius
       width=0
       color {colorUseRule=UseColor_Specified, specifiedColor=<rgba 0 0 0 0>}
       cornerRadii=<0 0 0 0>
      }
      tintColor {colorUseRule=UseColor_Specified, specifiedColor=<rgba 1 1 1 1>}
      uVRegion {bIsValid=false, max=<0 0>, min=<0 0>}
     }
     backgroundImageSelected {
      bIsDynamicallyLoaded=false
      drawAs=Image
      imageType=NoImage
      mirroring=NoMirror
      resourceName=None
      resourceObject=null
      tiling=NoTile
      imageSize=<0 0>
      margin=<ltrb 0 0 0 0>
      outlineSettings {
       bUseBrushTransparency=false
       roundingType=HalfHeightRadius
       width=0
       color {colorUseRule=UseColor_Specified, specifiedColor=<rgba 0 0 0 0>}
       cornerRadii=<0 0 0 0>
      }
      tintColor {colorUseRule=UseColor_Specified, specifiedColor=<rgba 0 0.162 0.7454 1>}
      uVRegion {bIsValid=false, max=<0 0>, min=<0 0>}
     }
     caretImage {
      bIsDynamicallyLoaded=false
      drawAs=Image
      imageType=NoImage
      mirroring=NoMirror
      resourceName=None
      resourceObject=null
      tiling=NoTile
      imageSize=<0 0>
      margin=<ltrb 0 0 0 0>
      outlineSettings {
       bUseBrushTransparency=false
       roundingType=HalfHeightRadius
       width=0
       color {colorUseRule=UseColor_Specified, specifiedColor=<rgba 0 0 0 0>}
       cornerRadii=<0 0 0 0>
      }
      tintColor {colorUseRule=UseColor_Specified, specifiedColor=<rgba 1 1 1 1>}
      uVRegion {bIsValid=false, max=<0 0>, min=<0 0>}
     }
     colorAndOpacity {colorUseRule=UseColor_Foreground, specifiedColor=<rgba 1 0 1 1>}
     font {
      bForceMonospaced=false
      bMaterialIsStencil=false
      fontMaterial=null
      fontObject=null
      letterSpacing=0
      monospacedWidth=1
      size=20
      skewAmount=0
      typefaceFontName=Regular
      outlineSettings {
       bApplyOutlineToDropShadows=false
       bMiteredCorners=false
       bSeparateFillAlpha=false
       outlineMaterial=null
       outlineSize=0
       outlineColor=<rgba 0 0 0 1>
      }
     }
     hintColorAndOpacity {colorUseRule=UseColor_Foreground, specifiedColor=<rgba 1 0 1 1>}
     selectedBackgroundColor {colorUseRule=UseColor_Foreground, specifiedColor=<rgba 1 0 1 1>}
    }}
   slot {class=BorderSlot, properties {padding=<ltrb 12 4 12 2>}}
  }
  HorizontalBox_3 {widgetClassPath=&HorizontalBox, parent=VerticalBox, widget=HorizontalBox_3, slot {class=VerticalBoxSlot}}
  TextBlock_4 {
   widgetClassPath=&TextBlock
   parent=HorizontalBox_3
   widget=TextBlock_4
   properties {
    text=Texture Size X
    font {
     bForceMonospaced=false
     bMaterialIsStencil=false
     fontMaterial=null
     fontObject=&Roboto
     letterSpacing=0
     monospacedWidth=1
     size=24
     skewAmount=0
     typefaceFontName=Regular
     outlineSettings {
      bApplyOutlineToDropShadows=false
      bMiteredCorners=false
      bSeparateFillAlpha=false
      outlineMaterial=null
      outlineSize=0
      outlineColor=<rgba 0 0 0 1>
     }
    }
    shadowColorAndOpacity=<rgba 0 0 0 1>
    shadowOffset=<2 2>
   }
   slot {class=HorizontalBoxSlot, properties {padding=<ltrb 5 5 5 5>}}
  }
  Border_10 {
   widgetClassPath=&Border
   parent=HorizontalBox_3
   widget=Border_10
   properties {
    background {
     bIsDynamicallyLoaded=false
     drawAs=RoundedBox
     imageType=NoImage
     mirroring=NoMirror
     resourceName=None
     resourceObject=null
     tiling=NoTile
     imageSize=<32 32>
     margin=<ltrb 0 0 0 0>
     outlineSettings {
      bUseBrushTransparency=false
      roundingType=HalfHeightRadius
      width=0
      color {colorUseRule=UseColor_Specified, specifiedColor=<rgba 0 0 0 0>}
      cornerRadii=<0 0 0 0>
     }
     tintColor {colorUseRule=UseColor_Specified, specifiedColor=<rgba 1 1 1 1>}
     uVRegion {bIsValid=false, max=<0 0>, min=<0 0>}
    }
    brushColor=<rgba 0.0823 0.0823 0.0823 1>
    padding=<ltrb 12 4 12 2>
   }
   slot {class=HorizontalBoxSlot, properties {padding=<ltrb 5 5 5 5>, size {sizeRule=Fill, value=1}}}
  }
  BuildTextureSizeX {
   widgetClassPath=&EditorUtilityEditableText
   bIsVariable=true
   parent=Border_10
   widget=BuildTextureSizeX
   properties {widgetStyle {
     backgroundImageComposing {
      bIsDynamicallyLoaded=false
      drawAs=Border
      imageType=FullColor
      mirroring=NoMirror
      resourceName=../../../Engine/Content/Slate/Old/HyperlinkDotted.png
      resourceObject=null
      tiling=Both
      imageSize=<0 0>
      margin=<ltrb 0 0 0 0.1875>
      outlineSettings {
       bUseBrushTransparency=false
       roundingType=HalfHeightRadius
       width=0
       color {colorUseRule=UseColor_Specified, specifiedColor=<rgba 0 0 0 0>}
       cornerRadii=<0 0 0 0>
      }
      tintColor {colorUseRule=UseColor_Specified, specifiedColor=<rgba 1 1 1 1>}
      uVRegion {bIsValid=false, max=<0 0>, min=<0 0>}
     }
     backgroundImageSelected {
      bIsDynamicallyLoaded=false
      drawAs=Image
      imageType=NoImage
      mirroring=NoMirror
      resourceName=None
      resourceObject=null
      tiling=NoTile
      imageSize=<0 0>
      margin=<ltrb 0 0 0 0>
      outlineSettings {
       bUseBrushTransparency=false
       roundingType=HalfHeightRadius
       width=0
       color {colorUseRule=UseColor_Specified, specifiedColor=<rgba 0 0 0 0>}
       cornerRadii=<0 0 0 0>
      }
      tintColor {colorUseRule=UseColor_Specified, specifiedColor=<rgba 0 0.162 0.7454 1>}
      uVRegion {bIsValid=false, max=<0 0>, min=<0 0>}
     }
     caretImage {
      bIsDynamicallyLoaded=false
      drawAs=Image
      imageType=NoImage
      mirroring=NoMirror
      resourceName=None
      resourceObject=null
      tiling=NoTile
      imageSize=<0 0>
      margin=<ltrb 0 0 0 0>
      outlineSettings {
       bUseBrushTransparency=false
       roundingType=HalfHeightRadius
       width=0
       color {colorUseRule=UseColor_Specified, specifiedColor=<rgba 0 0 0 0>}
       cornerRadii=<0 0 0 0>
      }
      tintColor {colorUseRule=UseColor_Specified, specifiedColor=<rgba 1 1 1 1>}
      uVRegion {bIsValid=false, max=<0 0>, min=<0 0>}
     }
     colorAndOpacity {colorUseRule=UseColor_Foreground, specifiedColor=<rgba 1 0 1 1>}
     font {
      bForceMonospaced=false
      bMaterialIsStencil=false
      fontMaterial=null
      fontObject=null
      letterSpacing=0
      monospacedWidth=1
      size=20
      skewAmount=0
      typefaceFontName=Regular
      outlineSettings {
       bApplyOutlineToDropShadows=false
       bMiteredCorners=false
       bSeparateFillAlpha=false
       outlineMaterial=null
       outlineSize=0
       outlineColor=<rgba 0 0 0 1>
      }
     }
     hintColorAndOpacity {colorUseRule=UseColor_Foreground, specifiedColor=<rgba 1 0 1 1>}
     selectedBackgroundColor {colorUseRule=UseColor_Foreground, specifiedColor=<rgba 1 0 1 1>}
    }}
   slot {class=BorderSlot, properties {padding=<ltrb 12 4 12 2>}}
  }
  TextBlock_5 {
   widgetClassPath=&TextBlock
   parent=HorizontalBox_3
   widget=TextBlock_5
   properties {
    text=Texture Size Y
    font {
     bForceMonospaced=false
     bMaterialIsStencil=false
     fontMaterial=null
     fontObject=&Roboto
     letterSpacing=0
     monospacedWidth=1
     size=24
     skewAmount=0
     typefaceFontName=Regular
     outlineSettings {
      bApplyOutlineToDropShadows=false
      bMiteredCorners=false
      bSeparateFillAlpha=false
      outlineMaterial=null
      outlineSize=0
      outlineColor=<rgba 0 0 0 1>
     }
    }
    shadowColorAndOpacity=<rgba 0 0 0 1>
    shadowOffset=<2 2>
   }
   slot {class=HorizontalBoxSlot, properties {padding=<ltrb 5 5 5 5>}}
  }
  Border_9 {
   widgetClassPath=&Border
   parent=HorizontalBox_3
   widget=Border_9
   properties {
    background {
     bIsDynamicallyLoaded=false
     drawAs=RoundedBox
     imageType=NoImage
     mirroring=NoMirror
     resourceName=None
     resourceObject=null
     tiling=NoTile
     imageSize=<32 32>
     margin=<ltrb 0 0 0 0>
     outlineSettings {
      bUseBrushTransparency=false
      roundingType=HalfHeightRadius
      width=0
      color {colorUseRule=UseColor_Specified, specifiedColor=<rgba 0 0 0 0>}
      cornerRadii=<0 0 0 0>
     }
     tintColor {colorUseRule=UseColor_Specified, specifiedColor=<rgba 1 1 1 1>}
     uVRegion {bIsValid=false, max=<0 0>, min=<0 0>}
    }
    brushColor=<rgba 0.0823 0.0823 0.0823 1>
    padding=<ltrb 12 4 12 2>
   }
   slot {class=HorizontalBoxSlot, properties {padding=<ltrb 5 5 5 5>, size {sizeRule=Fill, value=1}}}
  }
  BuildTextureSizeY {
   widgetClassPath=&EditorUtilityEditableText
   bIsVariable=true
   parent=Border_9
   widget=BuildTextureSizeY
   properties {widgetStyle {
     backgroundImageComposing {
      bIsDynamicallyLoaded=false
      drawAs=Border
      imageType=FullColor
      mirroring=NoMirror
      resourceName=../../../Engine/Content/Slate/Old/HyperlinkDotted.png
      resourceObject=null
      tiling=Both
      imageSize=<0 0>
      margin=<ltrb 0 0 0 0.1875>
      outlineSettings {
       bUseBrushTransparency=false
       roundingType=HalfHeightRadius
       width=0
       color {colorUseRule=UseColor_Specified, specifiedColor=<rgba 0 0 0 0>}
       cornerRadii=<0 0 0 0>
      }
      tintColor {colorUseRule=UseColor_Specified, specifiedColor=<rgba 1 1 1 1>}
      uVRegion {bIsValid=false, max=<0 0>, min=<0 0>}
     }
     backgroundImageSelected {
      bIsDynamicallyLoaded=false
      drawAs=Image
      imageType=NoImage
      mirroring=NoMirror
      resourceName=None
      resourceObject=null
      tiling=NoTile
      imageSize=<0 0>
      margin=<ltrb 0 0 0 0>
      outlineSettings {
       bUseBrushTransparency=false
       roundingType=HalfHeightRadius
       width=0
       color {colorUseRule=UseColor_Specified, specifiedColor=<rgba 0 0 0 0>}
       cornerRadii=<0 0 0 0>
      }
      tintColor {colorUseRule=UseColor_Specified, specifiedColor=<rgba 0 0.162 0.7454 1>}
      uVRegion {bIsValid=false, max=<0 0>, min=<0 0>}
     }
     caretImage {
      bIsDynamicallyLoaded=false
      drawAs=Image
      imageType=NoImage
      mirroring=NoMirror
      resourceName=None
      resourceObject=null
      tiling=NoTile
      imageSize=<0 0>
      margin=<ltrb 0 0 0 0>
      outlineSettings {
       bUseBrushTransparency=false
       roundingType=HalfHeightRadius
       width=0
       color {colorUseRule=UseColor_Specified, specifiedColor=<rgba 0 0 0 0>}
       cornerRadii=<0 0 0 0>
      }
      tintColor {colorUseRule=UseColor_Specified, specifiedColor=<rgba 1 1 1 1>}
      uVRegion {bIsValid=false, max=<0 0>, min=<0 0>}
     }
     colorAndOpacity {colorUseRule=UseColor_Foreground, specifiedColor=<rgba 1 0 1 1>}
     font {
      bForceMonospaced=false
      bMaterialIsStencil=false
      fontMaterial=null
      fontObject=null
      letterSpacing=0
      monospacedWidth=1
      size=20
      skewAmount=0
      typefaceFontName=Regular
      outlineSettings {
       bApplyOutlineToDropShadows=false
       bMiteredCorners=false
       bSeparateFillAlpha=false
       outlineMaterial=null
       outlineSize=0
       outlineColor=<rgba 0 0 0 1>
      }
     }
     hintColorAndOpacity {colorUseRule=UseColor_Foreground, specifiedColor=<rgba 1 0 1 1>}
     selectedBackgroundColor {colorUseRule=UseColor_Foreground, specifiedColor=<rgba 1 0 1 1>}
    }}
   slot {class=BorderSlot, properties {padding=<ltrb 12 4 12 2>}}
  }
  HorizontalBox_1 {widgetClassPath=&HorizontalBox, parent=VerticalBox, widget=HorizontalBox_1, slot {class=VerticalBoxSlot}}
  TextBlock_7 {
   widgetClassPath=&TextBlock
   parent=HorizontalBox_1
   widget=TextBlock_7
   properties {
    text=Texture Atlas Name
    font {
     bForceMonospaced=false
     bMaterialIsStencil=false
     fontMaterial=null
     fontObject=&Roboto
     letterSpacing=0
     monospacedWidth=1
     size=24
     skewAmount=0
     typefaceFontName=Regular
     outlineSettings {
      bApplyOutlineToDropShadows=false
      bMiteredCorners=false
      bSeparateFillAlpha=false
      outlineMaterial=null
      outlineSize=0
      outlineColor=<rgba 0 0 0 1>
     }
    }
    shadowColorAndOpacity=<rgba 0 0 0 1>
    shadowOffset=<2 2>
   }
   slot {class=HorizontalBoxSlot, properties {padding=<ltrb 5 5 5 5>}}
  }
  Border_1 {
   widgetClassPath=&Border
   parent=HorizontalBox_1
   widget=Border_1
   properties {
    background {
     bIsDynamicallyLoaded=false
     drawAs=RoundedBox
     imageType=NoImage
     mirroring=NoMirror
     resourceName=None
     resourceObject=null
     tiling=NoTile
     imageSize=<32 32>
     margin=<ltrb 0 0 0 0>
     outlineSettings {
      bUseBrushTransparency=false
      roundingType=HalfHeightRadius
      width=0
      color {colorUseRule=UseColor_Specified, specifiedColor=<rgba 0 0 0 0>}
      cornerRadii=<0 0 0 0>
     }
     tintColor {colorUseRule=UseColor_Specified, specifiedColor=<rgba 1 1 1 1>}
     uVRegion {bIsValid=false, max=<0 0>, min=<0 0>}
    }
    brushColor=<rgba 0.0823 0.0823 0.0823 1>
    padding=<ltrb 12 4 12 2>
   }
   slot {class=HorizontalBoxSlot, properties {padding=<ltrb 5 5 5 5>, size {sizeRule=Fill, value=1}}}
  }
  BuildTextureAtlasName {
   widgetClassPath=&EditorUtilityEditableText
   bIsVariable=true
   parent=Border_1
   widget=BuildTextureAtlasName
   properties {widgetStyle {
     backgroundImageComposing {
      bIsDynamicallyLoaded=false
      drawAs=Border
      imageType=FullColor
      mirroring=NoMirror
      resourceName=../../../Engine/Content/Slate/Old/HyperlinkDotted.png
      resourceObject=null
      tiling=Both
      imageSize=<0 0>
      margin=<ltrb 0 0 0 0.1875>
      outlineSettings {
       bUseBrushTransparency=false
       roundingType=HalfHeightRadius
       width=0
       color {colorUseRule=UseColor_Specified, specifiedColor=<rgba 0 0 0 0>}
       cornerRadii=<0 0 0 0>
      }
      tintColor {colorUseRule=UseColor_Specified, specifiedColor=<rgba 1 1 1 1>}
      uVRegion {bIsValid=false, max=<0 0>, min=<0 0>}
     }
     backgroundImageSelected {
      bIsDynamicallyLoaded=false
      drawAs=Image
      imageType=NoImage
      mirroring=NoMirror
      resourceName=None
      resourceObject=null
      tiling=NoTile
      imageSize=<0 0>
      margin=<ltrb 0 0 0 0>
      outlineSettings {
       bUseBrushTransparency=false
       roundingType=HalfHeightRadius
       width=0
       color {colorUseRule=UseColor_Specified, specifiedColor=<rgba 0 0 0 0>}
       cornerRadii=<0 0 0 0>
      }
      tintColor {colorUseRule=UseColor_Specified, specifiedColor=<rgba 0 0.162 0.7454 1>}
      uVRegion {bIsValid=false, max=<0 0>, min=<0 0>}
     }
     caretImage {
      bIsDynamicallyLoaded=false
      drawAs=Image
      imageType=NoImage
      mirroring=NoMirror
      resourceName=None
      resourceObject=null
      tiling=NoTile
      imageSize=<0 0>
      margin=<ltrb 0 0 0 0>
      outlineSettings {
       bUseBrushTransparency=false
       roundingType=HalfHeightRadius
       width=0
       color {colorUseRule=UseColor_Specified, specifiedColor=<rgba 0 0 0 0>}
       cornerRadii=<0 0 0 0>
      }
      tintColor {colorUseRule=UseColor_Specified, specifiedColor=<rgba 1 1 1 1>}
      uVRegion {bIsValid=false, max=<0 0>, min=<0 0>}
     }
     colorAndOpacity {colorUseRule=UseColor_Foreground, specifiedColor=<rgba 1 0 1 1>}
     font {
      bForceMonospaced=false
      bMaterialIsStencil=false
      fontMaterial=null
      fontObject=null
      letterSpacing=0
      monospacedWidth=1
      size=20
      skewAmount=0
      typefaceFontName=Regular
      outlineSettings {
       bApplyOutlineToDropShadows=false
       bMiteredCorners=false
       bSeparateFillAlpha=false
       outlineMaterial=null
       outlineSize=0
       outlineColor=<rgba 0 0 0 1>
      }
     }
     hintColorAndOpacity {colorUseRule=UseColor_Foreground, specifiedColor=<rgba 1 0 1 1>}
     selectedBackgroundColor {colorUseRule=UseColor_Foreground, specifiedColor=<rgba 1 0 1 1>}
    }}
   slot {class=BorderSlot, properties {padding=<ltrb 12 4 12 2>}}
  }
  BuildButton {widgetClassPath=&EditorUtilityButton, parent=VerticalBox, widget=BuildButton, slot {class=VerticalBoxSlot}}
  TextBlock_3 {widgetClassPath=&TextBlock, parent=BuildButton, widget=TextBlock_3, properties {text=Build}, slot {class=ButtonSlot}}
 ]
}
